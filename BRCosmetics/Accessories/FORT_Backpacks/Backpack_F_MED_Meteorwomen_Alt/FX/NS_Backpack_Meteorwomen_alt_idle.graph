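# Niagara idle effect for the alternate Meteorwomen back bling, adding persistent glows, lights and small tumbling meshes.
# Glow and tumbling-mesh emitters burst two particles once, re-sampled every update onto skeletal mesh bones or sockets indexed by spawn group.
# User.ColorSwap blends glow and light colors from cyan-blue toward green and sets the single-particle Mesh emitter's visibility tag.
# The main glow pulses alpha over looping lifetimes, scaled by time of day.
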NiagaraSystem /BRCosmetics/Accessories/FORT_Backpacks/Backpack_F_MED_Meteorwomen_Alt/FX/NS_Backpack_Meteorwomen_alt_idle
dependencies {
 usedDataInterfaces [
  &NiagaraDataInterfaceCurve
  &NiagaraDataInterfaceArrayFloat3
  &NiagaraDataInterfaceColorCurve
  &NiagaraDataInterfaceArrayInt32
 ]
 usedDynamicInputs [&MakeVectorFromLinearColorRGB]
 usedModules [&SystemState]
}
emitters {
 Glow001 {
  emitterUpdateScript [
   &EmitterState()
   &SpawnRate(SpawnRate=0.4, Use Spawn Probability=false) {enabled=false}
   &SpawnBurst_Instantaneous(Spawn Count=2, Spawn Probability=HlslExpression() {hlslExpression="Emitter.LoopCount == 0"}, Use Spawn Probability=true)
   &SpawnBurst_Instantaneous(Spawn Count=2, Spawn Time=8.7334, Use Spawn Probability=false, UseLoopCountLimit=false) {enabled=false, module=SpawnBurst_Instantaneous001}
  ]
  particleSpawnScript [
   &FN_SetSpawnGroup(Number of Spawn Groups=2) {module=FN_SetSpawnGroup001}
   &InitializeParticle(Lifetime=6, Sprite Size Mode=::NewEnumerator3, Uniform Sprite Size=40, UsePositionOffset=false, Write Material Random=true) {module=InitializeParticle001}
   &DynamicMaterialParameters(
    Index 0 Param 1=10
    Module.Param0WriteEnabled=true
    Module.Param1WriteEnabled=false
    Module.Param2WriteEnabled=false
    Module.Param3WriteEnabled=false
   ) {enabled=false}
   &BoxLocation(Box Offset=<0 -20 0>, Box Size=<0 0 0>, MaskSpawnGroup=false, Module.Override Seed=false, Rotate Box=false, Surface Only=false) {enabled=false, module=BoxLocation001}
  ]
  particleUpdateScript [
   &ParticleState(Kill Particles When Lifetime Has Elapsed=false)
   &Color(Color=&MultiplyLinearColor(
     Linear Color=&LerpLinearColors(StartColor=<rgba 0 1.8 4 1>, EndColor=<rgba 0 2 0.7333 1>, LerpFactor=$User.ColorSwap)
     Scale Factor=&MakeVector4_FromFloat(Float=0.35)
    ))
   &SkeletalMeshLocation(
    Bone / Socket Index=$Particles.SpawnGroup
    MaskSpawnGroup=false
    Mesh Sampling Type=::NewEnumerator5
    Sampled Position Offset=&SelectVectorFromArray(Direct Array Index=$Particles.SpawnGroup, Vector Selection Array=DataInterface() {propertyValues {FloatData [<-0.5 0.5 0>, <0.5 0.5 0>]}})
    Sampled Position Offset Coordinate Space:ENiagaraSkelMeshTransforms=::NewEnumerator4
    Skeletal Mesh=$User.SkeletalMesh
    Socket Sampling Mode=::NewEnumerator5
   ) {module=SkeletalMeshLocation001}
   &ScaleColor(
    Scale Alpha=&FloatFromCurve(FloatCurve=DataInterface() {propertyValues {Curve {keys [<at 0 0 RCIM_Cubic>,<at 0.2 1 RCIM_Cubic>,<at 1 0>]}}})
    Scale RGB=<1.5 1.5 1.2>
    ScaleA=true
    ScaleRGB=true
   ) {enabled=false}
   &CameraOffset(Camera Offset Amount=4)
   &ScaleSpriteSize(
    Uniform Curve Scale=0.25
    Uniform Curve Sprite Scale=DataInterface() {propertyValues {Curve {keys [{tangentMode=RCTM_User, time=1, value=1}]}}}
   )
  ]
  rendererClasses [&NiagaraSpriteRendererProperties]
  renderers [{rendererClass=&NiagaraSpriteRendererProperties}]
 }
 Glow_Scalability {
  emitterSpawnScript [&FN_PreserveSkydiveUserVariables() {module=PreserveUserVariables}]
  emitterUpdateScript [
   &EmitterState(Life Cycle Mode=::NewEnumerator1, Loop Behavior=::NewEnumerator1, Loop Duration Mode=::NewEnumerator1, UseLoopDelay=false)
   &SpawnBurst_Instantaneous(Spawn Count=2, Spawn Probability=HlslExpression() {hlslExpression="Emitter.LoopCount == 0"}, Use Spawn Probability=true)
  ]
  particleSpawnScript [
   &FN_SetSpawnGroup(Number of Spawn Groups=2)
   /Niagara/Modules/Spawn/Initialization/InitializeParticle(
    Color=&MultiplyLinearColor(
     Linear Color=&LerpLinearColors(StartColor=<rgba 0 1.8 4 1>, EndColor=<rgba 1.2667 4 0 1>, LerpFactor=$User.ColorSwap)
     Scale Factor=&MakeVector4_FromFloat(Float=1)
    )
    Lifetime=&RandomRangeFloat(Minimum=3, Maximum=3.5)
    Module.Write Color=false
    Module.Write Lifetime=true
    Module.Write Mass=false
    Module.Write Material Random=true
    Module.Write Position=true
    Module.Write Scale=true
    Module.Write SpriteRotation=false
    Module.Write SpriteSize=true
    Sprite Size=&Vector2DFromFloat(Value=200)
   )
   &CameraOffset(Camera Offset Amount=2)
   /SetVariables(Particles.SpriteColor=&MultiplyLinearColor(
     Linear Color=&LerpLinearColors(StartColor=<rgba 0 0.8167 1 1>, EndColor=<rgba 0 1 0.7333 1>, LerpFactor=$User.ColorSwap)
     Scale Factor=&MakeVector4_FromFloat(Float=&Lerp_Float(A=1, B=1, Alpha=$User.ColorSwap))
    )) {isSetParameters=true}
  ]
  particleUpdateScript [
   &ParticleState(
    Kill Particles When Lifetime Has Elapsed=false
    Let Infinitely Lived Particles Die When Emitter Deactivates=true
    Loop Particles Lifetime=true
   )
   &ScaleSpriteSize(
    Uniform Curve Scale=0.1
    Uniform Curve Sprite Scale=DataInterface() {propertyValues {Curve {keys [{tangentMode=RCTM_User, time=1, value=1}]}}}
   )
   &SkeletalMeshLocation(
    Bone / Socket Index=$Particles.SpawnGroup
    MaskSpawnGroup=false
    Mesh Sampling Type=::NewEnumerator5
    Sampled Position Offset=&SelectVectorFromArray(Direct Array Index=$Particles.SpawnGroup, Vector Selection Array=DataInterface() {propertyValues {FloatData [<0 -0.35 1.5>, <0 -0.35 1.5>]}})
    Sampled Position Offset Coordinate Space:ENiagaraSkelMeshTransforms=::NewEnumerator4
    Skeletal Mesh=$User.SkeletalMesh
    Socket Sampling Mode=::NewEnumerator5
   ) {module=SkeletalMeshLocation001}
   /SetVariables(Particles.SpritePosition:Vector3f=&Add_Vector(
     A=$Particles.SkeletalMeshLocation001.SampledPosition
     B=&SelectVectorFromArray(Direct Array Index=$Particles.SpawnGroup, Vector Selection Array=DataInterface() {propertyValues {FloatData [<0 0 1.5>, <0 0 1.5>]}})
    )) {isSetParameters=true}
   /SetVariables(Particles.SpriteColor=&MultiplyLinearColor(
     Linear Color=&LerpLinearColors(StartColor=<rgba 0 0.8167 1 1>, EndColor=<rgba 0 1 0.7333 1>, LerpFactor=$User.ColorSwap)
     Scale Factor=&MakeVector4_FromFloat(Float=&Lerp_Float(A=1, B=1, Alpha=$User.ColorSwap))
    )) {isSetParameters=true}
   &ScaleColor(
    Color Value To Scale=$Particles.SpriteColor
    Scale Mode=::NewEnumerator1
    Scale RGBA=&MakeVector4_FromLinearColor(LinearColor=&ColorFromCurve(ColorCurve=DataInterface() {propertyValues {
        AlphaCurve {keys [<at 0 0.25>,<at 0.5 0.5>,<at 1 0.25>]}
        BlueCurve {keys [<at 0 1 RCIM_Cubic>,<at 1 1 RCIM_Cubic>]}
        GreenCurve {keys [<at 0 1 RCIM_Cubic>,<at 1 1 RCIM_Cubic>]}
        RedCurve {keys [<at 0 1 RCIM_Cubic>,<at 1 1 RCIM_Cubic>]}
       }}))
    ScaleRGBA=true
   )
   &FN_Light_Attributes(
    Color=&LerpLinearColors(StartColor=<rgba 0 0.8167 1 1>, EndColor=<rgba 0 1 0.4 1>, LerpFactor=$User.ColorSwap)
    Create Color Attribute=true
    Create Exponent Attribute=true
    Create Radius Attribute=true
    Exponent=11
    Radius=4
    Scale Color=&Lerp_Float(A=1400, B=1500, Alpha=$User.ColorSwap)
   )
   &FN_Gen5_Light(Attenutation=60, Color:LinearColor=<rgba 0 0.4833 0.5 1>, Exponent=100, Intensity=500, Position Offset:Vector3f=<0 0 -10>)
   &ScaleColor(
    Color Value To Scale=$Particles.SpriteColor
    Scale Alpha=&Multiply_Float(A=40, B=&FNTimeOfDayScalar(Night Scale=1, Morning Scale=0.5, Day Scale=0.8, Evening Scale=0.5))
    ScaleA=true
    ScaleRGB=true
   ) {module=ScaleColor001}
  ]
  rendererClasses [
   &NiagaraSpriteRendererProperties
   &NiagaraComponentRendererProperties
   &NiagaraLightRendererProperties
  ]
  renderers [
   {rendererClass=&NiagaraSpriteRendererProperties}
   {rendererClass=&NiagaraComponentRendererProperties, rendererIndex=1}
   {rendererClass=&NiagaraLightRendererProperties, rendererIndex=2}
  ]
 }
 Glow_Scalability003 {
  emitterSpawnScript [&FN_PreserveSkydiveUserVariables() {module=PreserveUserVariables}]
  emitterUpdateScript [
   &EmitterState()
   &SpawnBurst_Instantaneous(
    Spawn Count=&MakeIntFromFloat(Float=2)
    Spawn Probability=HlslExpression() {hlslExpression="Emitter.LoopCount == 0"}
    Use Spawn Probability=true
   )
  ]
  particleSpawnScript [
   &InitializeParticle(
    Color=<rgba 1 1 1 0.2>
    Mesh Scale=&VectorFromFloat(Value=&Multiply_Float(A=1.8, B=0.05))
    Module.Write Color=true
    Module.Write Lifetime=true
    Module.Write Mass=false
    Module.Write Material Random=true
    Module.Write Position=false
    Module.Write Scale=true
    Module.Write SpriteRotation=false
    Module.Write SpriteSize=true
    Sprite Size=<19 19>
   )
   /SetVariables(Particles.MeshOrientation=&OrientMeshToVector(
     Look At Direction=&UniformRangedVector(Minimum=<-1 -1 -1>, Maximum=<1 1 1>)
     Reference Axis Vector=<1 0 0>
    )) {isSetParameters=true}
   /SetVariables(Particles.RotationalVelocity:Vector3f=&UniformRangedVector(
     Minimum=&VectorFromFloat(Value=-0.4)
     Maximum=&VectorFromFloat(Value=0.4)
    )) {isSetParameters=true}
   /SetVariables(Particles.SpritePosition:Vector3f=&Add_Vector(A=Linked() {linkedVariable {name=Particles.Position, type=NiagaraPosition}}, B=<0 0.5 -0.5>)) {isSetParameters=true}
   &FN_SetSpawnGroup(Number of Spawn Groups=2)
   &CameraOffset(Camera Offset Amount=-3)
  ]
  particleUpdateScript [
   &ParticleState(Kill Particles When Lifetime Has Elapsed=false, Let Infinitely Lived Particles Die When Emitter Deactivates=true)
   &SkeletalMeshLocation(
    Bone / Socket Index=$Particles.SpawnGroup
    MaskSpawnGroup=false
    Mesh Sampling Type=::NewEnumerator5
    Sampled Position Offset=&SelectVectorFromArray(Direct Array Index=$Particles.SpawnGroup, Vector Selection Array=DataInterface() {propertyValues {FloatData [<0 0.5 0.3>, <0 0.5 0.3>]}})
    Sampled Position Offset Coordinate Space:ENiagaraSkelMeshTransforms=::NewEnumerator4
    Skeletal Mesh=$User.SkeletalMesh
    Socket Sampling Mode=::NewEnumerator5
   ) {module=SkeletalMeshLocation001}
   &MeshRotationForce(Rotation=&UniformRangedVector(Minimum=<-3 -3 -3>, Maximum=<3 3 3>))
   &SolveForcesAndVelocity(Clamp Velocity=false, Limit Acceleration=false)
   &SolveRotationalForcesAndVelocity()
   /SetVariables(Particles.RotationLimited:Vector3f=&self:LimitVelocity_01(
     Max:Vector3f=&VectorFromFloat(Value=0.4)
     Min:Vector3f=&VectorFromFloat(Value=-0.4)
    )) {isSetParameters=true}
   /SetVariables(Particles.RotationalVelocity:Vector3f=$Particles.RotationLimited) {isSetParameters=true}
  ]
  rendererClasses [&NiagaraMeshRendererProperties]
  renderers [{rendererClass=&NiagaraMeshRendererProperties, rendererIndex=0}]
 }
 Mesh {
  emitterSpawnScript [&FN_PreserveSkydiveUserVariables() {module=PreserveUserVariables}]
  emitterUpdateScript [
   &EmitterState(Life Cycle Mode=::NewEnumerator1, Loop Behavior=::NewEnumerator1, Loop Duration Mode=::NewEnumerator1, UseLoopDelay=false)
   &SpawnBurst_Instantaneous(Spawn Count=1, Spawn Probability=HlslExpression() {hlslExpression="Emitter.LoopCount == 0"}, Use Spawn Probability=true)
  ]
  particleSpawnScript [
   &FN_SetSpawnGroup(Number of Spawn Groups=2)
   &InitializeParticle(
    Color=&MultiplyLinearColor(
     Linear Color=&LerpLinearColors(StartColor=<rgba 0 1.8 4 1>, EndColor=<rgba 1.2667 4 0 1>, LerpFactor=$User.ColorSwap)
     Scale Factor=&MakeVector4_FromFloat(Float=1)
    )
    Lifetime=&RandomRangeFloat(Minimum=3, Maximum=3.5)
    Module.Write Color=false
    Module.Write Lifetime=true
    Module.Write Mass=false
    Module.Write Material Random=true
    Module.Write Position=true
    Module.Write Scale=true
    Module.Write SpriteRotation=false
    Module.Write SpriteSize=true
    Sprite Size=&Vector2DFromFloat(Value=200)
   )
  ]
  particleUpdateScript [
   &ParticleState(
    Kill Particles When Lifetime Has Elapsed=false
    Let Infinitely Lived Particles Die When Emitter Deactivates=true
    Loop Particles Lifetime=true
   )
   &ScaleMeshSize(Scale Factor=&VectorFromFloat(Value=1))
   &ScaleSpriteSize(
    Uniform Curve Scale=0.1
    Uniform Curve Sprite Scale=DataInterface() {propertyValues {Curve {keys [{tangentMode=RCTM_User, time=1, value=1}]}}}
   )
   &SkeletalMeshLocation(
    Bone / Socket Index=$Particles.SpawnGroup
    MaskSpawnGroup=false
    Mesh Sampling Type=::NewEnumerator5
    Sampled Position Offset=&SelectVectorFromArray(Direct Array Index=$Particles.SpawnGroup, Vector Selection Array=DataInterface() {propertyValues {FloatData [<0 -0.35 1.5>, <0 -0.35 1.5>]}})
    Sampled Position Offset Coordinate Space:ENiagaraSkelMeshTransforms=::NewEnumerator4
    Skeletal Mesh=$User.SkeletalMesh
    Socket Sampling Mode=::NewEnumerator5
   ) {enabled=false, module=SkeletalMeshLocation001}
   /SetVariables(Particles.VisibilityTag=&SelectIntFromArray(
     Direct Array Index=&MakeIntFromFloat(Float=$User.ColorSwap)
     Int Selection Array=DataInterface() {propertyValues {IntData [0,1]}}
    )) {isSetParameters=true}
  ]
  rendererClasses [&NiagaraMeshRendererProperties]
  renderers [
   {rendererClass=&NiagaraMeshRendererProperties, rendererIndex=0}
   {rendererClass=&NiagaraMeshRendererProperties, rendererIndex=1}
   {rendererClass=&NiagaraMeshRendererProperties, rendererIndex=2}
  ]
 }
}
summary {
 systemName=NS_Backpack_Meteorwomen_alt_idle
 emitters [
  {
   emitterName=Glow_Scalability
   rendererClasses [
    &NiagaraSpriteRendererProperties
    &NiagaraComponentRendererProperties
    &NiagaraLightRendererProperties
   ]
  }
  {emitterName=Mesh, rendererClasses [&NiagaraMeshRendererProperties]}
  {emitterName=Glow001, rendererClasses [&NiagaraSpriteRendererProperties]}
  {emitterName=Glow_Scalability003, rendererClasses [&NiagaraMeshRendererProperties]}
 ]
 userVariables [
  User.bIsFrontEndPreview {type=NiagaraBool}
  User.ColorSwap {type=NiagaraFloat, defaultValue=0, description=""}
  User.IsSkydiving {type=NiagaraBool}
  User.NewLinearColor {type=LinearColor, description="", defaultValue {struct=LinearColor, value=<rgba 1 1 1 1>}}
  User.PlayerSpeed {type=NiagaraFloat, defaultValue=0, description=""}
  User.SkeletalMesh {
   type=NiagaraDataInterfaceSkeletalMesh
   defaultValue {
    struct=NiagaraExt_VariableValue_DataInterface
    value {dataInterface=&self:NiagaraDataInterfaceSkeletalMesh_0}
   }
  }
 ]
}
